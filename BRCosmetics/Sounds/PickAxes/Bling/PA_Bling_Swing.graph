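# Sound cue for the swing sound of the Bling pickaxe cosmetic.
# A Random node picks among five equally weighted swing wave variations.
# It plays quietly in the impacts sound class, with a pickaxe swing property sheet and priming on load.
# Overrides set logarithmic attenuation with low-pass filtering and listener focus, and cap concurrency at a few voices, stopping the quietest.
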SoundCue /BRCosmetics/Sounds/PickAxes/Bling/PA_Bling_Swing
root=Random_1
nodes [
 {id=Random_1, children [WavePlayer_0,WavePlayer_6,WavePlayer_7,WavePlayer_8,WavePlayer_9], properties {weights [1,1,1,1,1]}}
 {id=WavePlayer_0, properties {soundWaveAssetPtr=&/PA_Bling_Swing_01}}
 {id=WavePlayer_6, properties {soundWaveAssetPtr=&/PA_Bling_Swing_02}}
 {id=WavePlayer_7, properties {soundWaveAssetPtr=&/PA_Bling_Swing_03}}
 {id=WavePlayer_8, properties {soundWaveAssetPtr=&/PA_Bling_Swing_04}}
 {id=WavePlayer_9, properties {soundWaveAssetPtr=&/PA_Bling_Swing_05}}
]
properties {
 attenuationSettings=/Game/Sounds/Attenuation/Fort_Player_Footsteps_Att
 audioPropertiesSheet=/BRCosmetics/Sounds/Settings/PropertySheets/Pickaxe/APS_Cosmetic_Pickaxe_Swing
 bOverrideAttenuation=true
 bOverrideConcurrency=true
 bPrimeOnLoad=true
 maxDistance=1100
 soundClassObject=&Fort_Impacts
 volumeMultiplier=0.3
 attenuationOverrides {bAttenuateWithLPF=true, bEnableListenerFocus=true, distanceAlgorithm=Logarithmic, falloffDistance=800, attenuationShapeExtents {x=300}}
 concurrencyOverrides {maxCount=3, resolutionRule=StopQuietest}
}
